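# Material function in the Papaya environment, exposed to the library, that quantizes a 0-1 input into evenly spaced steps set by a Tiles count.
# It multiplies the input by Tiles, floors it, then scales it by the reciprocal of Tiles to snap to each step's lower edge.
# An Apply Offset static switch, on by default, adds half a step so values land at step centers.
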MaterialFunction /Game/Environments/Papaya/Materials/MaterialFunctions/MF_Papaya_Quantize_Scalar
graph {
 properties {bExposeToLibrary=true}
 nodes [
  n1 = FunctionInput() {inputName="Param (0-1)"}
  n2 = FunctionInput() {inputName=Tiles, inputType=FunctionInput_Scalar, sortPriority=10}
  n3 = Divide(B=@n2)
  n4 = Multiply(A=Floor(None=Multiply(A=@n1, B=@n2)), B=@n3)
  n5 = Multiply(A=@n3) {constB=0.5}
  n6 = StaticBool() {value=true}
  FunctionOutput(None=StaticSwitch(
    True=Add(A=@n4, B=@n5)
    False=@n4
    Value=FunctionInput(Preview=@n6) {bUsePreviewValueAsDefault=true, inputName=Apply Offset, inputType=FunctionInput_StaticBool, sortPriority=30}
   ))
 ]
}
