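# Material function for the Papaya event that builds audio-reactive, grid-quantized equalizer bar masks.
# Weights EventAudio_Spectrum bands by a curve-atlas row at a quantized lookup, then blends toward average EventAudio_Amplitude.
# That level thresholds a tile-quantized ramp modulated by noise, whose Y lookup it also shifts; a static switch selects solid or ramp-shaded bars.
# Outputs the strip-noise-modulated value, top-edge mask, raw amplitude, ramp and both noise samples.
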MaterialFunction /Game/Environments/Papaya/Materials/MaterialFunctions/MF_Papaya_Equalizer_Data
graph {
 properties {bExposeToLibrary=true}
 nodes [
  n1 = FunctionInput(Preview=Constant2Vector() {g=1, r=1}) {bUsePreviewValueAsDefault=true, inputName=Noise Scale, inputType=FunctionInput_Vector2, sortPriority=33}
  n2 = FunctionInput(Preview=TextureCoordinate()) {bUsePreviewValueAsDefault=true, inputName=TexCoord Override, inputType=FunctionInput_Vector2, sortPriority=70}
  n3 = MaterialFunctionCall(Float2=Multiply(A=@n2, B=@n1)) {materialFunction=&BreakOutFloat2Components}
  n4 = Constant() {r=10}
  n5 = FunctionInput(Preview=MaterialFunctionCall(X=@n4, Y=@n4) {materialFunction=&MakeFloat2}.Result) {bUsePreviewValueAsDefault=true, inputName=Grid Tiles X/Y, inputType=FunctionInput_Vector2, sortPriority=40}
  n6 = MaterialFunctionCall(Float2=Divide(A=@n5, B=@n1)) {materialFunction=&BreakOutFloat2Components}
  n7 = MaterialFunctionCall!("Param (0-1)"=@n3.R, Tiles=@n6.R, Apply Offset) {materialFunction=&/MF_Papaya_Quantize_Scalar}
  n8 = CollectionParameter() {collection=&FortniteMaterialParameters, parameterName=EventAudio_Spectrum}
  n9 = MaterialFunctionCall(Float4=@n8) {materialFunction=&BreakOutFloat4Components}
  n10 = Constant()
  n11 = CurveAtlasRowParameter(CurveTime=MaterialFunctionCall(
    "Param (0-1)"=FunctionInput(Preview=@n10) {bUsePreviewValueAsDefault=true, inputName=Audio Lookup Param, inputType=FunctionInput_Scalar}
    Tiles=MaterialFunctionCall(Float2=@n5) {materialFunction=&BreakOutFloat2Components}.R
   ) {materialFunction=&MF_Papaya_Quantize_Scalar}.Result) {
   atlas=/Game/Athena/Sounds/Events/Papaya/Curve_Atlas_Papaya
   curve=/Game/Athena/Sounds/Events/Papaya/C_Papaya_FrequencyMix_Util
   defaultValue=4
   group=MF_Papaya_Equalizer_Data
   parameterName=Frequency Mix Util Curve
  }
  n12 = Add(A=Multiply(A=@n9.R, B=@n11.R), B=Multiply(A=@n9.G, B=@n11.G))
  n13 = Multiply(A=@n9.B, B=@n11.B)
  n14 = CollectionParameter() {collection=&FortniteMaterialParameters, parameterName=EventAudio_Amplitude}
  n15 = LinearInterpolate(
   A=Add(A=Add(A=@n12, B=@n13), B=Multiply(A=@n9.A, B=@n11.A))
   B=@n14
   Alpha=FunctionInput(Preview=Constant()) {bUsePreviewValueAsDefault=true, inputName=Average Amplitude Mix, inputType=FunctionInput_Scalar, sortPriority=10}
  )
  n16 = Constant() {r=1}
  n17 = MaterialFunctionCall(
   "Param (0-1)"=Multiply(
    A=@n15
    B=FunctionInput(Preview=@n16) {bUsePreviewValueAsDefault=true, inputName=Grid Speed, inputType=FunctionInput_Scalar, sortPriority=50}
   )
   Tiles=@n6.G
   Apply Offset=StaticBool()
  ) {materialFunction=&MF_Papaya_Quantize_Scalar}
  n18 = Add(A=@n3.G, B=@n17.Result)
  n19 = TextureObjectParameter() {
   parameterName=Noise Texture
   samplerType=SAMPLERTYPE_LinearGrayscale
   texture=/Game/Global/Textures/Default/Noise/EqualizerComponents_v2_Noise
  }
  n20 = FunctionInput(Preview=@n19) {bUsePreviewValueAsDefault=true, inputName=Noise Texture, inputType=FunctionInput_Texture2D, sortPriority=30}
  n21 = TextureSample(
   UVs=MaterialFunctionCall(
    X=@n7.Result
    Y=MaterialFunctionCall("Param (0-1)"=@n18, Tiles=@n6.G) {materialFunction=&MF_Papaya_Quantize_Scalar}.Result
   ) {materialFunction=&MakeFloat2}.Result
   Tex=@n20
  ) {samplerType=SAMPLERTYPE_LinearGrayscale}
  n22 = Multiply(
   A=OneMinus(None=FunctionOutput(None=@n21.R) {outputName=EQ Noise, sortPriority=10})
   B=FunctionInput(Preview=Constant() {r=0.2}) {bUsePreviewValueAsDefault=true, inputName=Noise Gain, inputType=FunctionInput_Scalar, sortPriority=35}
  ) {constB=0.1}
  n23 = MaterialFunctionCall(Float2=@n2) {materialFunction=&BreakOutFloat2Components}
  n24 = MaterialFunctionCall(Float2=@n5) {materialFunction=&BreakOutFloat2Components}
  n25 = MaterialFunctionCall("Param (0-1)"=@n23.R, Tiles=@n24.R) {materialFunction=&MF_Papaya_Quantize_Scalar}
  n26 = MaterialFunctionCall("Param (0-1)"=@n23.G, Tiles=@n24.G) {materialFunction=&MF_Papaya_Quantize_Scalar}
  n27 = TextureObjectParameter() {
   parameterName=Ramp Texture
   samplerType=SAMPLERTYPE_LinearGrayscale
   texture=/Game/Global/Textures/Default/Masks_and_Gradients/EqualizerComponents_v2_Ramp
  }
  n28 = FunctionInput(Preview=@n27) {bUsePreviewValueAsDefault=true, inputName=Ramp Texture, inputType=FunctionInput_Texture2D, sortPriority=20}
  n29 = Power(
   Base=TextureSample(
    UVs=MaterialFunctionCall(X=@n25.Result, Y=@n26.Result) {materialFunction=&MakeFloat2}.Result
    Tex=@n28
   ) {samplerType=SAMPLERTYPE_LinearGrayscale}.R
   Exp=FunctionInput(Preview=Constant() {r=1}) {bUsePreviewValueAsDefault=true, inputName=Ramp Exponent, inputType=FunctionInput_Scalar, sortPriority=25}
  ) {constExponent=0.5}
  n30 = Divide(B=@n24.G)
  n31 = OneMinus(None=@n15)
  n32 = MaterialFunctionCall!(
   Input=Multiply(A=OneMinus(None=@n22), B=FunctionOutput(None=@n29) {outputName=EQ Ramp, sortPriority=5})
   Input Low=@n31
   Input High=Add(A=@n31, B=Multiply(A=@n30) {constB=0.95}) {constB=0.001}
   Target Low=Constant()
   Target Height=Constant() {r=1}
  ) {materialFunction=&RemapValueRange}
  n33 = MaterialFunctionCall(Float2=@n2) {materialFunction=&BreakOutFloat2Components}
  n34 = Divide(B=Add(A=@n24.G) {constB=0.0001})
  n35 = TextureSample(
   UVs=MaterialFunctionCall(X=@n33.R, Y=Multiply(A=@n34, B=@n33.G)) {materialFunction=&MakeFloat2}.Result
   Tex=@n28
  ) {samplerType=SAMPLERTYPE_LinearGrayscale}
  n36 = StaticBool() {value=true}
  n37 = Saturate(None=StaticSwitch(
    True=@n32.Result
    False=Multiply(A=@n32.Result, B=@n35.R)
    Value=FunctionInput(Preview=@n36) {bUsePreviewValueAsDefault=true, inputName=Output Solid Bars, inputType=FunctionInput_StaticBool, sortPriority=60}
   ))
  n38 = TextureSample(
   UVs=MaterialFunctionCall(X=@n7.Result, Y=@n17.Result) {materialFunction=&MakeFloat2}.Result
   Tex=@n20
  ) {samplerType=SAMPLERTYPE_LinearGrayscale}
  n39 = Multiply(
   A=OneMinus(None=FunctionOutput(None=@n38.R) {outputName=EQ Noise Strips, sortPriority=20})
   B=FunctionInput(Preview=Constant() {r=0.2}) {bUsePreviewValueAsDefault=true, inputName=Strip Noise Gain, inputType=FunctionInput_Scalar, sortPriority=37}
  ) {constB=0.1}
  FunctionOutput(None=Multiply(A=OneMinus(None=@n39), B=@n37)) {outputName=EQ Value}
  n40 = Abs(None=Subtract(A=@n37) {constB=0.5})
  FunctionOutput(None=OneMinus(None=Multiply(A=@n40) {constB=2})) {outputName=EQ Top Edge, sortPriority=30}
  FunctionOutput(None=@n15) {outputName=EQ Raw Amplitude, sortPriority=60}
 ]
}
