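# MF_TexelToUV is a library-exposed material function that converts an integer texel coordinate into a texture UV.
# It floors both Vector2 inputs, Texture Resolution and Texel Coordinate, to whole numbers.
# It adds half the reciprocal of the resolution to the texel, then divides by the resolution to output texcoord.
# Because that offset precedes the division, samples sit just past the texel's corner instead of at its center.
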MaterialFunction /Game/Environments/Papaya/Materials/MaterialFunctions/MF_TexelToUV
graph {
 properties {bExposeToLibrary=true}
 nodes [
  n1 = Floor(None=FunctionInput() {inputName=Texture Resolution, inputType=FunctionInput_Vector2, previewValue=<656 369 0 1>})
  n2 = Multiply(A=Divide(B=@n1)) {constB=0.5}
  n3 = Floor(None=FunctionInput() {inputName=Texel Coordinate, inputType=FunctionInput_Vector2, sortPriority=10})
  FunctionOutput(None=Divide(A=Add(A=@n2, B=@n3), B=@n1)) {outputName=texcoord}
 ]
}
