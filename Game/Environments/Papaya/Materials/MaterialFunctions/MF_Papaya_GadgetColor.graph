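# Material function that produces the color for Papaya event gadgets from media video and event color curves.
# It blends the video toward a base color, itself mixed from an RGB curve atlas row and a single color-code texel sampled from the video, then blends toward an intermission curve color and desaturates.
# Most inputs default to FortniteMaterialParameters collection values, allowing global control.
# An inverse-tonemap call on that texel does not feed the output.
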MaterialFunction /Game/Environments/Papaya/Materials/MaterialFunctions/MF_Papaya_GadgetColor
graph {
 properties {bExposeToLibrary=true}
 nodes [
  n1 = TextureCoordinate()
  n2 = TextureObjectParameter() {group=MF_Papaya_GadgetColor, parameterName=Media Texture, texture=/Game/Movies/Season5/Papaya_MediaPlayer_Video2}
  n3 = CollectionParameter() {collection=&FortniteMaterialParameters, parameterName=PapayaBaseColorMix}
  n4 = CollectionParameter() {
   collection=&FortniteMaterialParameters
   parameterName=PapayaBaseColorParameter
  }
  n5 = CurveAtlasRowParameter(CurveTime=Frac(None=FunctionInput(Preview=@n4) {bUsePreviewValueAsDefault=true, inputName=Base Color Parameter, inputType=FunctionInput_Scalar, sortPriority=10})) {
   atlas=/Game/Athena/Sounds/Events/Papaya/Curve_Atlas_Papaya
   curve=/Game/Athena/Sounds/Events/Papaya/C_Papaya_RGBColors
   defaultValue=5
   group=MF_Papaya_GadgetColor
   parameterName=Base Color Curve
   sortPriority=10
  }
  n6 = CollectionParameter() {
   collection=&FortniteMaterialParameters
   parameterName=PapayaMediaTextureResolution
  }
  n7 = FunctionInput(Preview=ComponentMask(None=@n6) {g=true, r=true}) {
   bUsePreviewValueAsDefault=true
   description=Defaults to 656x369
   inputName=Media Texture Resolution
   inputType=FunctionInput_Vector2
   sortPriority=5
  }
  n8 = CollectionParameter() {
   collection=&FortniteMaterialParameters
   parameterName=PapayaMediaColorCodeCoordinate
  }
  n9 = FunctionInput(Preview=ComponentMask(None=@n8) {g=true, r=true}) {
   bUsePreviewValueAsDefault=true
   description="Texel coordinate in Media to sample for Base Color. Defaults to 1,1"
   inputName=Media Color Code Coordinate
   inputType=FunctionInput_Vector2
   sortPriority=5
  }
  n10 = MaterialFunctionCall!(Texture Resolution=@n7, Texel Coordinate=@n9) {materialFunction=&/MF_TexelToUV}
  n11 = CollectionParameter() {
   collection=&FortniteMaterialParameters
   parameterName=PapayaMediaColorCodeMix
  }
  n12 = LinearInterpolate(
   A=TextureSample(UVs=FunctionInput(Preview=@n1) {bUsePreviewValueAsDefault=true, inputName=Media Player TexCoord, inputType=FunctionInput_Vector2}, Tex=@n2).RGB
   B=LinearInterpolate(
    A=@n5
    B=TextureSample!(UVs=@n10.texcoord, Tex=@n2, Level, Apply View MipBias) {automaticViewMipBias=false, constMipValue=0, mipValueMode=TMVM_MipLevel}.RGB
    Alpha=FunctionInput(Preview=@n11) {bUsePreviewValueAsDefault=true, inputName=Media Color Code Mix, inputType=FunctionInput_Scalar, sortPriority=25}
   )
   Alpha=FunctionInput(Preview=@n3) {bUsePreviewValueAsDefault=true, inputName=Base Color Mix, inputType=FunctionInput_Scalar, sortPriority=20}
  )
  n13 = CollectionParameter() {
   collection=&FortniteMaterialParameters
   parameterName=PapayaIntermissionColorParameter
  }
  n14 = CurveAtlasRowParameter(CurveTime=Frac(None=FunctionInput(Preview=@n13) {bUsePreviewValueAsDefault=true, inputName=Intermission Color Parameter, inputType=FunctionInput_Scalar, sortPriority=30})) {
   atlas=&Curve_Atlas_Papaya
   curve=/Game/Athena/Sounds/Events/Papaya/C_Papaya_IntermissionColors
   defaultValue=2
   group=MF_Papaya_GadgetColor
   parameterName=Intermission Color Curve
   sortPriority=30
  }
  n15 = CollectionParameter() {
   collection=&FortniteMaterialParameters
   parameterName=PapayaIntermissionColorMix
  }
  n16 = FunctionInput(Preview=@n15) {bUsePreviewValueAsDefault=true, inputName=Intermission Color Mix, inputType=FunctionInput_Scalar, sortPriority=40}
  n17 = CollectionParameter() {collection=&FortniteMaterialParameters, parameterName=PapayaDesaturation}
  FunctionOutput(None=Desaturation(
    None=LinearInterpolate(A=@n12, B=@n14, Alpha=@n16)
    Fraction=FunctionInput(Preview=@n17) {bUsePreviewValueAsDefault=true, inputName=Desaturation, inputType=FunctionInput_Scalar, sortPriority=50}
   ))
  MaterialFunctionCall!(UVs=@n10.texcoord, Texture=@n2) {materialFunction=/Game/Movies/InversePostProcess/MF_invertFNtonemap}
 ]
}
