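# MF_UVMirror is a library-exposed material function that tiles UVs with alternate repeats mirrored on each axis.
# It multiplies texcoord by Repeat Count, then uses Fmod by 2 to find which repeat each axis is in.
# Per-axis If nodes output the fraction when that remainder is below 1 and one minus the fraction when above.
# MakeFloat2 recombines both axes into the output UV.
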MaterialFunction /Game/Environments/Papaya/Materials/MaterialFunctions/MF_UVMirror
graph {
 properties {bExposeToLibrary=true}
 nodes [
  n1 = Multiply(
   A=FunctionInput() {inputName=texcoord, inputType=FunctionInput_Vector2}
   B=FunctionInput() {inputName=Repeat Count, inputType=FunctionInput_Vector2}
  )
  n2 = MaterialFunctionCall(Float2=Fmod(A=@n1, B=Constant() {r=2})) {materialFunction=&BreakOutFloat2Components}
  n3 = Frac(None=@n1)
  n4 = MaterialFunctionCall(Float2=OneMinus(None=@n3)) {materialFunction=&BreakOutFloat2Components}
  n5 = MaterialFunctionCall(Float2=@n3) {materialFunction=&BreakOutFloat2Components}
  FunctionOutput(None=MaterialFunctionCall(X=If(A=@n2.R, "A > B"=@n4.R, "A < B"=@n5.R) {constB=1}, Y=If(A=@n2.G, "A > B"=@n4.G, "A < B"=@n5.G) {constB=1}) {materialFunction=&MakeFloat2}.Result)
 ]
}
